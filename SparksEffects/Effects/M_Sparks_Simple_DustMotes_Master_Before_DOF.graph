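# Translucent, lit Niagara particle material for dust motes, drawn before depth of field with volumetric directional lighting.
# Particle color tints the texture, and the result passes through an emissive adjustments function.
# Opacity combines particle alpha with texture alpha raised to a power, then applies near-camera and depth fades, with power and fade distances driven by dynamic parameters.
# A sphere normal map, tiled by U and V parameters, gives motes rounded shading.
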Material /SparksEffects/Effects/M_Sparks_Simple_DustMotes_Master_Before_DOF
graph {
 properties {
  bCastRayTracedShadows=false
  bUsedWithNiagaraRibbons=true
  bUsedWithNiagaraSprites=true
  blendMode=BLEND_Translucent
  pixelDepthOffsetMode=PDOM_Legacy
  translucencyLightingMode=TLM_VolumetricDirectional
  translucencyPass=MTP_BeforeDOF
  usedShadingModels=Default Lit
 }
 nodes [
  n1 = ParticleColor()
  n2 = TextureSampleParameter2D() {parameterName=Texture, texture=/Game/Effects/Fort_Effects/Textures/T_DustMotes_4x2_01}
  n3 = DynamicParameter() {desc="R:(TextureOpacity)\r\nG: no notes\r\nB: no notes", defaultValue=<rgba 3 1 1 0>, paramNames [Power, DepthFade, CameraFade, NOT USED]}
  n4 = Power(Base=@n2.A, Exp=@n3.Power)
  DepthFade(
   Opacity=Multiply(
    A=Multiply(A=@n1.A, B=@n4)
    B=MaterialFunctionCall(Distance=@n3.CameraFade) {materialFunction=&Near_Camera_Fade}.Result
   )
   FadeDistance=@n3.DepthFade
  )
  MaterialFunctionCall(Emissive=Multiply(A=@n1.RGB, B=@n2.RGB)) {materialFunction=&MF_FX_EmissiveAdjustments}
  n5 = ScalarParameter() {defaultValue=4, parameterName=TileUNormal}
  n6 = ScalarParameter() {defaultValue=2, parameterName=TileVNormal}
  TextureSample(UVs=Multiply(A=TextureCoordinate(), B=AppendVector(A=@n5, B=@n6))) {samplerType=SAMPLERTYPE_Normal, texture=/Game/Environments/World/Backgrounds/Universal/TechContent/Textures/T_Spherenormal_01}
 ]
}
